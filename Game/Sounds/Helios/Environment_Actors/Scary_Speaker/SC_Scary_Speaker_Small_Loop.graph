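# Sound cue for the small Scary Speaker environment prop, looping intermittent one-shot sounds as ambient audio.
# A Looping node repeats a random 10 to 30 second delay, then a quality switch feeds one of two equal-weight random pools.
# The first two quality slots share the larger pool and the last two share a smaller, overlapping one.
# Small-speaker attenuation and a concurrency override that stops the farthest instance limit how many are heard.
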SoundCue /Game/Sounds/Helios/Environment_Actors/Scary_Speaker/SC_Scary_Speaker_Small_Loop
root=Looping_0
nodes [
 {id=Delay_0, children [QualityLevel_0], properties {delayMax=30, delayMin=10}}
 {id=Looping_0, children [Delay_0]}
 {id=QualityLevel_0, children [Random_1,Random_1,Random_2,Random_2]}
 {
  id=Random_1
  children [
   WavePlayer_0
   WavePlayer_1
   WavePlayer_2
   WavePlayer_3
   WavePlayer_4
   WavePlayer_5
   WavePlayer_6
   WavePlayer_7
   WavePlayer_8
   WavePlayer_9
   WavePlayer_10
   WavePlayer_11
   WavePlayer_12
   WavePlayer_13
   WavePlayer_14
   WavePlayer_15
   WavePlayer_16
   WavePlayer_17
   WavePlayer_18
   WavePlayer_19
   WavePlayer_20
   WavePlayer_21
   WavePlayer_22
   WavePlayer_23
   WavePlayer_24
   WavePlayer_25
   WavePlayer_26
   WavePlayer_27
   WavePlayer_28
   WavePlayer_29
   WavePlayer_30
   WavePlayer_31
  ]
  properties {bShouldExcludeFromBranchCulling=true, weights [1,1,1,1,1,1,1,1,1,1,1,1,1,1,1,1,1,1,1,1,1,1,1,1,1,1,1,1,1,1,1,1]}
 }
 {
  id=Random_2
  children [
   WavePlayer_34
   WavePlayer_35
   WavePlayer_36
   WavePlayer_38
   WavePlayer_39
   WavePlayer_41
   WavePlayer_42
   WavePlayer_43
   WavePlayer_44
   WavePlayer_45
   WavePlayer_46
   WavePlayer_47
   WavePlayer_48
   WavePlayer_58
   WavePlayer_59
   WavePlayer_60
   WavePlayer_61
   WavePlayer_62
   WavePlayer_63
   WavePlayer_64
   WavePlayer_65
   WavePlayer_66
   WavePlayer_67
  ]
  properties {bShouldExcludeFromBranchCulling=true, weights [1,1,1,1,1,1,1,1,1,1,1,1,1,1,1,1,1,1,1,1,1,1,1]}
 }
 {id=WavePlayer_0, properties {soundWaveAssetPtr=&/FNM_Scary_Speaker_Sounds_05}}
 {id=WavePlayer_1, properties {soundWaveAssetPtr=&/FNM_Scary_Speaker_Sounds_06}}
 {id=WavePlayer_10, properties {soundWaveAssetPtr=&/FNM_Scary_Speaker_Sounds_15}}
 {id=WavePlayer_11, properties {soundWaveAssetPtr=&/FNM_Scary_Speaker_Sounds_16}}
 {id=WavePlayer_12, properties {soundWaveAssetPtr=&/FNM_Scary_Speaker_Sounds_17}}
 {id=WavePlayer_13, properties {soundWaveAssetPtr=&/FNM_Scary_Speaker_Sounds_18}}
 {id=WavePlayer_14, properties {soundWaveAssetPtr=&/FNM_Scary_Speaker_Sounds_19}}
 {id=WavePlayer_15, properties {soundWaveAssetPtr=&/FNM_Scary_Speaker_Sounds_20}}
 {id=WavePlayer_16, properties {soundWaveAssetPtr=&/FNM_Scary_Speaker_Sounds_21}}
 {id=WavePlayer_17, properties {soundWaveAssetPtr=&/FNM_Scary_Speaker_Sounds_22}}
 {id=WavePlayer_18, properties {soundWaveAssetPtr=&/FNM_Scary_Speaker_Sounds_23}}
 {id=WavePlayer_19, properties {soundWaveAssetPtr=&/FNM_Scary_Speaker_Sounds_24}}
 {id=WavePlayer_2, properties {soundWaveAssetPtr=&/FNM_Scary_Speaker_Sounds_07}}
 {id=WavePlayer_20, properties {soundWaveAssetPtr=&/FNM_Scary_Speaker_Sounds_01}}
 {id=WavePlayer_21, properties {soundWaveAssetPtr=&/FNM_Scary_Speaker_Sounds_02}}
 {id=WavePlayer_22, properties {soundWaveAssetPtr=&/FNM_Scary_Speaker_Sounds_03}}
 {id=WavePlayer_23, properties {soundWaveAssetPtr=&/FNM_Scary_Speaker_Sounds_04}}
 {id=WavePlayer_24, properties {soundWaveAssetPtr=&/FNM_Scary_Speaker_Sounds_JB_01}}
 {id=WavePlayer_25, properties {soundWaveAssetPtr=&/FNM_Scary_Speaker_Sounds_JB_02}}
 {id=WavePlayer_26, properties {soundWaveAssetPtr=&/FNM_Scary_Speaker_Sounds_JB_03}}
 {id=WavePlayer_27, properties {soundWaveAssetPtr=&/FNM_Scary_Speaker_Sounds_JB_04}}
 {id=WavePlayer_28, properties {soundWaveAssetPtr=&/FNM_Scary_Speaker_Sounds_JB_05}}
 {id=WavePlayer_29, properties {soundWaveAssetPtr=&/FNM_Scary_Speaker_Sounds_JB_06}}
 {id=WavePlayer_3, properties {soundWaveAssetPtr=&/FNM_Scary_Speaker_Sounds_08}}
 {id=WavePlayer_30, properties {soundWaveAssetPtr=&/FNM_Scary_Speaker_Sounds_JB_07}}
 {id=WavePlayer_31, properties {soundWaveAssetPtr=&/FNM_Scary_Speaker_Sounds_JB_08}}
 {id=WavePlayer_34, properties {soundWaveAssetPtr=&FNM_Scary_Speaker_Sounds_05}}
 {id=WavePlayer_35, properties {soundWaveAssetPtr=&FNM_Scary_Speaker_Sounds_06}}
 {id=WavePlayer_36, properties {soundWaveAssetPtr=&FNM_Scary_Speaker_Sounds_07}}
 {id=WavePlayer_38, properties {soundWaveAssetPtr=&/FNM_Scary_Speaker_Sounds_09}}
 {id=WavePlayer_39, properties {soundWaveAssetPtr=&/FNM_Scary_Speaker_Sounds_10}}
 {id=WavePlayer_4, properties {soundWaveAssetPtr=&FNM_Scary_Speaker_Sounds_09}}
 {id=WavePlayer_41, properties {soundWaveAssetPtr=&/FNM_Scary_Speaker_Sounds_12}}
 {id=WavePlayer_42, properties {soundWaveAssetPtr=&/FNM_Scary_Speaker_Sounds_13}}
 {id=WavePlayer_43, properties {soundWaveAssetPtr=&/FNM_Scary_Speaker_Sounds_14}}
 {id=WavePlayer_44, properties {soundWaveAssetPtr=&FNM_Scary_Speaker_Sounds_15}}
 {id=WavePlayer_45, properties {soundWaveAssetPtr=&FNM_Scary_Speaker_Sounds_16}}
 {id=WavePlayer_46, properties {soundWaveAssetPtr=&FNM_Scary_Speaker_Sounds_17}}
 {id=WavePlayer_47, properties {soundWaveAssetPtr=&FNM_Scary_Speaker_Sounds_18}}
 {id=WavePlayer_48, properties {soundWaveAssetPtr=&FNM_Scary_Speaker_Sounds_19}}
 {id=WavePlayer_5, properties {soundWaveAssetPtr=&FNM_Scary_Speaker_Sounds_10}}
 {id=WavePlayer_58, properties {soundWaveAssetPtr=&FNM_Scary_Speaker_Sounds_JB_01}}
 {id=WavePlayer_59, properties {soundWaveAssetPtr=&FNM_Scary_Speaker_Sounds_JB_02}}
 {id=WavePlayer_6, properties {soundWaveAssetPtr=&/FNM_Scary_Speaker_Sounds_11}}
 {id=WavePlayer_60, properties {soundWaveAssetPtr=&FNM_Scary_Speaker_Sounds_JB_03}}
 {id=WavePlayer_61, properties {soundWaveAssetPtr=&FNM_Scary_Speaker_Sounds_JB_04}}
 {id=WavePlayer_62, properties {soundWaveAssetPtr=&FNM_Scary_Speaker_Sounds_JB_05}}
 {id=WavePlayer_63, properties {soundWaveAssetPtr=&FNM_Scary_Speaker_Sounds_JB_06}}
 {id=WavePlayer_64, properties {soundWaveAssetPtr=&FNM_Scary_Speaker_Sounds_JB_07}}
 {id=WavePlayer_65, properties {soundWaveAssetPtr=&FNM_Scary_Speaker_Sounds_JB_08}}
 {id=WavePlayer_66, properties {soundWaveAssetPtr=&/FNM_Scary_Speaker_Sounds_JB_09}}
 {id=WavePlayer_67, properties {soundWaveAssetPtr=&/FNM_Scary_Speaker_Sounds_JB_10}}
 {id=WavePlayer_7, properties {soundWaveAssetPtr=&FNM_Scary_Speaker_Sounds_12}}
 {id=WavePlayer_8, properties {soundWaveAssetPtr=&FNM_Scary_Speaker_Sounds_13}}
 {id=WavePlayer_9, properties {soundWaveAssetPtr=&FNM_Scary_Speaker_Sounds_14}}
]
properties {
 attenuationSettings=&/ATT_Scary_Speaker_Small
 bOverrideConcurrency=true
 maxDistance=13300
 soundClassObject=&Fort_Ambient_Props_Athena
 volumeMultiplier=1.5
 concurrencyOverrides {maxCount=4, resolutionRule=StopFarthestThenPreventNew}
}
